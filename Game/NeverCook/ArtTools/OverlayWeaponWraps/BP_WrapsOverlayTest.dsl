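; Art-tools test Blueprint that previews weapon wraps as overlay materials on a rifle and a sniper skeletal mesh.
; The ApplyOverlayWrap custom event picks a wrap material instance from the MaterialSelection variable and passes it to SetMaterialParams for both weapons.
; SetMaterialParams creates a dynamic instance of the wrap, copies bounds vectors and the customization mask from the mesh's own material, then sets it as the mesh overlay.
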(event EventBeginPlay)

(event Custom|ApplyOverlayWrap
  (bind _materialselection (Variables|Default|GetMaterialSelection))
  (bind _returnvalue (Utilities|Select :Index _materialselection :NewEnumerator1 "/Game/NeverCook/ArtTools/OverlayWeaponWraps/MaterialInstances/MI_WeaponWrapOverlay_SnowKnight.MI_WeaponWrapOverlay_SnowKnight" :NewEnumerator2 "/Game/NeverCook/ArtTools/OverlayWeaponWraps/MaterialInstances/MI_WeaponWrapOverlay_SmartHyena.MI_WeaponWrapOverlay_SmartHyena" :NewEnumerator3 "/Game/NeverCook/ArtTools/OverlayWeaponWraps/MaterialInstances/MI_WeaponWrapOverlay_UnknownNetwork.MI_WeaponWrapOverlay_UnknownNetwork"))
  (CallFunction|SetMaterialParams (Variables|Default|GetRifle) _returnvalue)
  (CallFunction|SetMaterialParams (Variables|Default|GetSniper) _returnvalue))

(event Collision|EventActorBeginOverlap
  (params (OtherActor :type Actor%20Object%20Reference)))

(event EventTick
  (params (DeltaSeconds :type Float%20%28single-precision%29)))

(fn SetMaterialParams
  (params (SkelMesh :type Skeletal%20Mesh%20Component%20Object%20Reference) (OverlayMaterial :type Material%20Interface%20Object%20Reference))
  (bind _material_instance_dynamic (Class|BTechArtGlobalLibrary|GetDynamicMaterialInstance SkelMesh))
  (bind _returnvalue (Rendering|Material|CreateDynamicMaterialInstance OverlayMaterial "None" "None" :!class /Script/Engine.KismetMaterialLibrary))
  (bind _returnvalue_2 (Rendering|Material|GetVectorParameterValue _material_instance_dynamic "DO NOT OVERRIDE-OBCENTER" :!class /Script/Engine.MaterialInstanceDynamic))
  (bind _returnvalue_3 (Rendering|Material|GetVectorParameterValue _material_instance_dynamic "DO NOT OVERRIDE-OBMIN" :!class /Script/Engine.MaterialInstanceDynamic))
  (bind _returnvalue_4 (Rendering|Material|GetVectorParameterValue _material_instance_dynamic "DO NOT OVERRIDE-OBMAX" :!class /Script/Engine.MaterialInstanceDynamic))
  (bind _returnvalue_5 (Rendering|Material|GetTextureParameterValue _material_instance_dynamic "CustomizationMask (_CM)" :!class /Script/Engine.MaterialInstanceDynamic))
  (Rendering|Material|SetVectorParameterValue _returnvalue "DO NOT OVERRIDE-OBCENTER" _returnvalue_2 :!class /Script/Engine.MaterialInstanceDynamic)
  (Rendering|Material|SetVectorParameterValue _returnvalue "DO NOT OVERRIDE-OBMIN" _returnvalue_3 :!class /Script/Engine.MaterialInstanceDynamic)
  (Rendering|Material|SetVectorParameterValue _returnvalue "DO NOT OVERRIDE-OBMAX" _returnvalue_4 :!class /Script/Engine.MaterialInstanceDynamic)
  (Rendering|Material|SetTextureParameterValue _returnvalue "CustomizationMask (_CM)" _returnvalue_5)
  (Rendering|Material|SetOverlayMaterial SkelMesh _returnvalue false 0 :!class /Script/Engine.MeshComponent))

(fn ConstructionScript)
