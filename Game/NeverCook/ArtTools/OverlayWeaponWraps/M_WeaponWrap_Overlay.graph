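# Masked overlay material that draws cosmetic wraps on weapon meshes as a shell pushed out along vertex normals by WPO_Push.
# MF_Base_Customization builds the wrap over a dummy master, using customization masks, rotation and scale parameters, and per-weapon bounds, safe-zone and pivot vectors marked DO NOT OVERRIDE.
# The result passes through MF_TechArtOverrides_accessories, then gets its opacity mask and world position offset set.
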Material /Game/NeverCook/ArtTools/OverlayWeaponWraps/M_WeaponWrap_Overlay
graph {
 properties {
  bAutomaticallySetUsageInEditor=false
  bUseMaterialAttributes=true
  bUsedWithClothing=true
  bUsedWithMeshDeformer=true
  bUsedWithMorphTargets=true
  bUsedWithSkeletalMesh=true
  blendMode=BLEND_Masked
  pixelDepthOffsetMode=PDOM_Legacy
  refractionMethod=RM_IndexOfRefraction
  usedShadingModels=Default Lit
 }
 nodes [
  n1 = VectorParameter() {group=Weapon Specific Wrap Adjustments, parameterName=DO NOT OVERRIDE-SZ-Back, sortPriority=100, defaultValue=<rgba 0.3 0 0 0.2>}
  NamedRerouteDeclaration(None=AppendVector(A=@n1.RGB, B=@n1.RGB)) {name=SZ_Back, nodeColor=<rgba 1 0 0.0703 1>}
  n2 = VectorParameter() {group=Weapon Specific Wrap Adjustments, parameterName=DO NOT OVERRIDE-SZ-Middle, sortPriority=101, defaultValue=<rgba 0 0 0 0.2>}
  NamedRerouteDeclaration(None=AppendVector(A=@n2.RGB, B=@n2.RGB)) {name=SZ_Middle, nodeColor=<rgba 1 0 0.0703 1>}
  n3 = VectorParameter() {group=Weapon Specific Wrap Adjustments, parameterName=DO NOT OVERRIDE-SZ-Front, sortPriority=102, defaultValue=<rgba -0.3 0 0 0.2>}
  NamedRerouteDeclaration(None=AppendVector(A=@n3.RGB, B=@n3.RGB)) {name=SZ_Front, nodeColor=<rgba 1 0 0.0703 1>}
  n4 = VectorParameter() {group=Weapon Specific Wrap Adjustments, parameterName=DO NOT OVERRIDE-Virtual-Pivot, sortPriority=110, defaultValue=<rgba 0 0 0 0.5>}
  NamedRerouteDeclaration(None=AppendVector(A=@n4.RGB, B=@n4.RGB)) {name=SZ_VP, nodeColor=<rgba 1 0 0.0703 1>}
  n5 = VectorParameter() {group=Weapon Specific Wrap Adjustments, parameterName=DO NOT OVERRIDE-OBMAX, sortPriority=50, defaultValue=<rgba 0 0 0 1>}
  NamedRerouteDeclaration(None=@n5.RGB) {name=OBMAX, nodeColor=<rgba 0 1 0.6484 1>}
  n6 = VectorParameter() {group=Weapon Specific Wrap Adjustments, parameterName=DO NOT OVERRIDE-OBMIN, sortPriority=51, defaultValue=<rgba 0 0 0 1>}
  NamedRerouteDeclaration(None=@n6.RGB) {name=OBMIN, nodeColor=<rgba 0 1 0.6484 1>}
  n7 = VectorParameter() {group=Weapon Specific Wrap Adjustments, parameterName=DO NOT OVERRIDE-OBCENTER, sortPriority=55, defaultValue=<rgba 0 0 0 1>}
  NamedRerouteDeclaration(None=@n7.RGB) {name=OBCENTER, nodeColor=<rgba 0 1 0.6484 1>}
  n8 = MaterialFunctionCall!(UV 1 Coordinates=TextureCoordinate()) {materialFunction=/Game/Packages/Fortress_SharedMaterials/Base_Material_Functions/BaseCustomizationFunctions/MF_Base_CustomizationMask}
  NamedRerouteDeclaration(None=@n8."Customization Mask 1 (_CM R)") {name=OverlayAlpha, nodeColor=<rgba 1 0.9609 0 1>}
  n9 = MaterialFunctionCall() {materialFunction=/Game/Packages/Fortress_SharedMaterials/Base_Material_Functions/BaseCustomizationFunctions/MF_FN_Dummy_MASTER}
  n10 = MaterialFunctionCall!(
   Material=@n9.Material Result
   ------------------
   Object Bounds Center Adjustment=NamedRerouteUsage()
   Object Bounds Min Adjustment=NamedRerouteUsage()
   Wrap ScaleMultiplier=ScalarParameter() {defaultValue=1, group=Primary Triplaner, parameterName=Vehicle Wrap Scale Multiplier, sortPriority=-49}
   Rotation X Axis=ScalarParameter() {group=Primary Triplaner, parameterName=Global X Rot, sortPriority=-39}
   "Customization Mask(_CM A)"=@n8."Customization GMAP (_CM A)"
   Forward Axis
   Object Bounds Max Adjustment=NamedRerouteUsage()
   Rotation Z Axis=Add(
    A=ScalarParameter() {group=Primary Triplaner, parameterName=Vehicle Z axis Rotation, sortPriority=-32}
    B=ScalarParameter() {group=Primary Triplaner, parameterName=Global Z Rot}
   )
   Normal Map=@n9.Material Result
   Cropped Emissive Mask=@n9.Material Result
   "Customization Crevasse (_CM G)"=@n8."Customization GMAP (_CM A)"
   "Customization Mask UV Islands (_CM B Bottom Right)"=@n8."Customization GMAP (_CM A)"
   -------------------
   "Customization Mask 1 (_CM B Top Left)"=@n8."Customization GMAP (_CM A)"
   "Customization Mask 2 (_CM B Top Right)"=@n8."Customization GMAP (_CM A)"
   "Customization Mask 3 (_CM B Bottom Left)"=@n8."Customization GMAP (_CM A)"
   "Customization Mask (_CM RGBA)"=@n8."Customization GMAP (_CM A)"
   --------------------
   Pre-Skinned Local Position=@n9.Material Result
   Pre-Skinned Bounds=@n9.Material Result
   Pre-Skinned Local Normal=@n9.Material Result
   Per_ObjectMaskOverride
   isVehicle
   Side Project Vehicle
   isMech
   isCharacter
   isCharAccent
   isPickaxe
   isBackpack
   Object-SafeZone-Back=NamedRerouteUsage()
   Object-SafeZone-Middle=NamedRerouteUsage()
   Object-SafeZone-Front=NamedRerouteUsage()
   Object-SafeZone-VP_Charcter_Preview_Switch
   Object-SafeZone-Virtual-Pivot=NamedRerouteUsage()
   Wrap Holdout
  ) {materialFunction=/Game/Packages/Fortress_SharedMaterials/Base_Material_Functions/BaseCustomizationFunctions/MF_Base_Customization}
  n11 = MaterialFunctionCall(In=@n10.Result) {materialFunction=&MF_TechArtOverrides_accessories}
  SetMaterialAttributes!(
   MaterialAttributes=@n11.Result
   Opacity Mask=NamedRerouteUsage()
   World Position Offset=Multiply(A=VertexNormalWS(), B=ScalarParameter() {defaultValue=0.1, parameterName=WPO_Push})
  ) {attributeSetTypes [#1, #2]}
  ShadingPathSwitch(Default=Constant(), Deferred=@n10.Result)
 ]
}
